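# Physics asset for the frontend mesh of the F_MED_CraftGlue female player body, covering dynamic bandana and scarf-tie bones.
# Capsule bodies on the upper spine, clavicles and upper arms form the torso volume around the cloth chains.
# The bandana branches left and right from a middle bone chained to spine_05, while two two-bone scarf ties hang directly from spine_05.
# Cloth joints have tight, asymmetric swing limits with zero twist; only the spine_04-to-spine_03 joint is free.
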PhysicsAsset /BRCosmetics/Characters/Player/Female/Medium/Bodies/F_MED_CraftGlue/Meshes/F_MED_CraftGlue_Frontend_Physics
bodies {
 clavicle_l [{shapeType=Capsule, length=6.5, radius=5, shapeName=clavicle_l_capsule, center=<6.3661 0 0>, rotation=<rot 90 -90 -90.0003>}]
 clavicle_r [{shapeType=Capsule, length=6.5, radius=5, shapeName=clavicle_r_capsule, center=<-6.3661 0 0>, rotation=<rot -90 90 -89.9998>}]
 dyn_bandana_l_01 [{shapeType=Capsule, length=3.5, radius=2.5, shapeName=dyn_bandana_l_01_capsule, center=<2.763 0 1.5431>, rotation=<rot 0 -90 90.0006>}]
 dyn_bandana_l_02 [{shapeType=Capsule, length=3.5, radius=2.5, shapeName=dyn_bandana_l_02_capsule, center=<2.763 0 1.5431>, rotation=<rot 0 -90 90.0005>}]
 dyn_bandana_mid_01 [{shapeType=Capsule, length=3.5, radius=2.5, shapeName=dyn_bandana_mid_01_capsule, center=<2.763 0 1.5431>, rotation=<rot 0 -90 90.0006>}]
 dyn_bandana_mid_02 [{shapeType=Capsule, length=3.5, radius=2.5, shapeName=dyn_bandana_mid_02_capsule, center=<3.6016 0 1.5431>, rotation=<rot 0 -90 90.0006>}]
 dyn_bandana_r_01 [{shapeType=Capsule, length=3.5, radius=2.5, shapeName=dyn_bandana_r_01_capsule, center=<2.763 0 1.5431>, rotation=<rot 0 -90 90.0006>}]
 dyn_bandana_r_02 [{shapeType=Capsule, length=3.5, radius=2.5, shapeName=dyn_bandana_r_02_capsule, center=<2.763 0 1.5431>, rotation=<rot 0 -90 90.0006>}]
 dyn_scarfTie_l_01 [{shapeType=Capsule, length=3.5, radius=1.5, shapeName=dyn_scarfTie_l_01_capsule, center=<2.3811 0 0>, rotation=<rot 90 0 0>}]
 dyn_scarfTie_l_02 [{shapeType=Capsule, length=3, radius=2.5, shapeName=dyn_scarfTie_l_02_capsule, center=<2.3811 0 0>, rotation=<rot 90 0 0>}]
 dyn_scarfTie_r_01 [{shapeType=Capsule, length=3.5, radius=1.5, shapeName=dyn_scarfTie_r_01_capsule, center=<2.3811 0 0>, rotation=<rot 90 0 0>}]
 dyn_scarfTie_r_02 [{shapeType=Capsule, length=3, radius=2.5, shapeName=dyn_scarfTie_r_02_capsule, center=<2.3811 0 0>, rotation=<rot 90 0 0>}]
 spine_03 [{shapeType=Capsule, length=6.5, radius=11.5, shapeName=spine_03_capsule, center=<0 1.6101 0>}]
 spine_04 [{shapeType=Capsule, length=12.5, radius=11.5, shapeName=spine_04_capsule, center=<-1.0223 0.893 0>}]
 spine_05 [{shapeType=Capsule, length=14, radius=10, shapeName=spine_05_capsule, center=<-0.2727 1.2155 0>}]
 upperarm_l [{shapeType=Capsule, length=12, radius=7, shapeName=upperarm_l_capsule, center=<6.6527 0 0>, rotation=<rot 90 -166.7373 -166.7377>}]
 upperarm_r [{shapeType=Capsule, length=12.5, radius=6.5, shapeName=upperarm_r_capsule, center=<-6.6527 -0.6578 0>, rotation=<rot -90 90 -90>}]
}
constraints [
 {bone1Name=spine_05, bone2Name=spine_04, swing1LimitDegrees=45, swing2LimitDegrees=45, twistLimitDegrees=0}
 {bone1Name=clavicle_r, bone2Name=spine_05, swing1LimitDegrees=45, swing2LimitDegrees=45, twistLimitDegrees=0}
 {bone1Name=clavicle_l, bone2Name=spine_05, swing1LimitDegrees=45, swing2LimitDegrees=45, twistLimitDegrees=0}
 {bone1Name=upperarm_l, bone2Name=clavicle_l, swing1LimitDegrees=45, swing2LimitDegrees=45, twistLimitDegrees=0}
 {bone1Name=upperarm_r, bone2Name=clavicle_r, swing1LimitDegrees=45, swing2LimitDegrees=45, twistLimitDegrees=0}
 {
  bone1Name=spine_04
  bone2Name=spine_03
  swing1LimitDegrees=45
  swing1Motion=Free
  swing2LimitDegrees=45
  swing2Motion=Free
  twistLimitDegrees=45
  twistMotion=Free
 }
 {bone1Name=dyn_bandana_mid_01, bone2Name=spine_05, swing1LimitDegrees=0, swing2LimitDegrees=30, twistLimitDegrees=0}
 {bone1Name=dyn_bandana_l_01, bone2Name=dyn_bandana_mid_01, swing1LimitDegrees=0, swing2LimitDegrees=30, twistLimitDegrees=0}
 {bone1Name=dyn_bandana_l_02, bone2Name=dyn_bandana_l_01, swing1LimitDegrees=10, swing2LimitDegrees=45, twistLimitDegrees=0}
 {bone1Name=dyn_bandana_mid_02, bone2Name=dyn_bandana_mid_01, swing1LimitDegrees=10, swing2LimitDegrees=40, twistLimitDegrees=0}
 {bone1Name=dyn_bandana_r_01, bone2Name=dyn_bandana_mid_01, swing1LimitDegrees=0, swing2LimitDegrees=30, twistLimitDegrees=0}
 {bone1Name=dyn_bandana_r_02, bone2Name=dyn_bandana_r_01, swing1LimitDegrees=10, swing2LimitDegrees=45, twistLimitDegrees=0}
 {bone1Name=dyn_scarfTie_l_01, bone2Name=spine_05, swing1LimitDegrees=20, swing2LimitDegrees=30, twistLimitDegrees=0}
 {bone1Name=dyn_scarfTie_l_02, bone2Name=dyn_scarfTie_l_01, swing1LimitDegrees=20, swing2LimitDegrees=30, twistLimitDegrees=0}
 {bone1Name=dyn_scarfTie_r_01, bone2Name=spine_05, swing1LimitDegrees=20, swing2LimitDegrees=30, twistLimitDegrees=0}
 {bone1Name=dyn_scarfTie_r_02, bone2Name=dyn_scarfTie_r_01, swing1LimitDegrees=20, swing2LimitDegrees=30, twistLimitDegrees=0}
]
